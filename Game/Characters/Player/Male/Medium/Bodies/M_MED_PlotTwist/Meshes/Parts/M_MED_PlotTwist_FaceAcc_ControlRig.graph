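# Control Rig for the M_MED_PlotTwist face accessory that offsets dynamic hair root bones as the head tilts.
# IsCrouching, then IsSkydiving, selects one of three functions: crouching, skydiving or default hair.
# Each uses spherical pose readers on the head, relative to spine_05, pelvis or neck_02 respectively, to weight ModifyTransforms with variable-driven offsets.
# A helper mirrors some side-hair offsets by negating translation; the head_ctrl-to-head SetTransform is left out of execution.
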ControlRigBlueprint /Game/Characters/Player/Male/Medium/Bodies/M_MED_PlotTwist/Meshes/Parts/M_MED_PlotTwist_FaceAcc_ControlRig
graphs [
 ContainedGraph {nodes [
   RigVMFunctionEntryNode()
   > SequenceAggregate = RigVMAggregateNode()
   n1 = &RigUnit_SphericalPoseReader(
    ActiveRegionScaleFactors=(PositiveWidth=1,NegativeWidth=3,PositiveHeight=1,NegativeHeight=1)
    ActiveRegionSize=0.05
    Debug=(bDrawDebug=true,DebugScale=25,DebugSegments=20,DebugThickness=0.25)
    DriverAxis=(X=1,Y=0,Z=0)
    DriverItem=(Type=Bone,Name="head")
    FalloffRegionScaleFactors=(PositiveWidth=1,NegativeWidth=1,PositiveHeight=1,NegativeHeight=1)
    FalloffSize=0.3
    OptionalParentItem=(Type=Bone,Name="neck_02")
    RotationOffset=(X=-90,Y=0,Z=20)
    ExecutePin=@SequenceAggregate.A
   )
   n2 = RigVMVariableNode() {variable=local_rear_hair_left_45_head_bk}
   > Deform_GetMirrorTransform_FNC_1_1 = RigVMFunctionReferenceNode(transform=@n2.Value)
   > &RigUnit_ModifyTransforms(
    ItemToModify="((Item=(Type=Bone),Transform=(Rotation=(X=0,Y=0,Z=0,W=1),Translation=(X=0,Y=0,Z=0),Scale3D=(X=1,Y=1,Z=1))),(Item=(Type=Bone,Name=\"dyn_rear_hair_01_root\"),Transform=(Rotation=(X=0,Y=0,Z=0,W=1),Translation=(X=0,Y=0,Z=0),Scale3D=(X=1,Y=1,Z=1))),(Item=(Type=Bo ...(truncated)"
    ItemToModify.3.Transform=@Deform_GetMirrorTransform_FNC_1_1.mirrorTransform
    Weight=@n1.OutputParam
    ItemToModify.1.Transform=RigVMVariableNode() {variable=local_rear_hair_head_bk}.Value
    ItemToModify.2.Transform=@n2.Value
   )
   n3 = &RigUnit_SphericalPoseReader(
    ActiveRegionScaleFactors=(PositiveWidth=2,NegativeWidth=2,PositiveHeight=2,NegativeHeight=2)
    ActiveRegionSize=0.05
    Debug=(bDrawDebug=true,DebugScale=25,DebugSegments=20,DebugThickness=0.25)
    DriverAxis=(X=1,Y=0,Z=0)
    DriverItem=(Type=Bone,Name="head")
    FalloffRegionScaleFactors=(PositiveWidth=1,NegativeWidth=1,PositiveHeight=1,NegativeHeight=1)
    FalloffSize=0.3
    OptionalParentItem=(Type=Bone,Name="neck_02")
    RotationOffset=(X=-30,Y=-50,Z=0)
    ExecutePin=@SequenceAggregate.B
   )
   > &RigUnit_ModifyTransforms(
    ItemToModify="((Item=(Type=Bone),Transform=(Rotation=(X=0,Y=0,Z=0,W=1),Translation=(X=0,Y=0,Z=0),Scale3D=(X=1,Y=1,Z=1))),(Item=(Type=Bone,Name=\"dyn_rear_hair_01_root\"),Transform=(Rotation=(X=0,Y=0,Z=0,W=1),Translation=(X=0,Y=0,Z=0),Scale3D=(X=1,Y=1,Z=1))),(Item=(Type=Bo ...(truncated)"
    Weight=@n3.OutputParam
    ItemToModify.2.Transform=RigVMVariableNode() {variable=ModifyTransforms_1_ItemToModify_2_Transform}.Value
    ItemToModify.3.Transform=RigVMVariableNode() {variable=ModifyTransforms_1_ItemToModify_3_Transform}.Value
   )
   RigVMFunctionReturnNode(ExecuteContext=@SequenceAggregate.D)
   Multiply = &RigVMFunction_MathDoubleMul(A=@n1.OutputParam, B=RigVMVariableNode() {variable=UseHeadBackSPR}.Value)
   RigVMVariableNode() {variable=RotationOffset}
   RigVMVariableNode() {variable=SPR Parent}
  ]}
 ContainedGraph {nodes [
   n1 = RigVMFunctionEntryNode()
   SequenceAggregate = &RigVMFunction_Sequence(ExecuteContext=@n1.ExecuteContext)
   SequenceAggregate_1 = &RigVMFunction_Sequence(ExecuteContext=@SequenceAggregate.B)
   SequenceAggregate_2 = &RigVMFunction_Sequence(ExecuteContext=@SequenceAggregate_1.B)
   RigVMFunctionReturnNode(A=@SequenceAggregate.A, B=@SequenceAggregate_1.A, C=@SequenceAggregate_2.A, D=@SequenceAggregate_2.B)
  ]}
 ContainedGraph {nodes [
   n1 = RigVMFunctionEntryNode()
   n2 = /Script/RigVM.RigVMFunction_MathTransformToEulerTransform(Value=@n1.transform) {is=struct RigVMUnitNode, template="Cast::Execute(in Value,out Result)"}
   > RigVMFunctionReturnNode(
    mirrorTransform=(Rotation=(X=0,Y=0,Z=0,W=1),Translation=(X=0,Y=0,Z=0),Scale3D=(X=0,Y=0,Z=0))
    mirrorTransform.Rotation=/Script/RigVM.RigVMFunction_MathQuaternionFromRotatorV2(Value=@n2.Result.Rotation) {is=struct RigVMUnitNode, template="Cast::Execute(in Value,out Result)"}.Result
    mirrorTransform.Translation=/Script/RigVM.RigVMFunction_MathVectorScale(Factor=-1, Value=@n2.Result.Location) {is=struct RigVMUnitNode, template="Scale::Execute(in Value,in Factor,out Result)"}.Result
    mirrorTransform.Scale3D=@n2.Result.Scale
   )
  ]}
 ContainedGraph {nodes [
   RigVMFunctionEntryNode()
   > SequenceAggregate = RigVMAggregateNode()
   n1 = &RigUnit_SphericalPoseReader(
    ActiveRegionScaleFactors=(PositiveWidth=1.5,NegativeWidth=1.3,PositiveHeight=1,NegativeHeight=2)
    ActiveRegionSize=0.05
    Debug=(bDrawDebug=true,DebugScale=25,DebugSegments=20,DebugThickness=0.25)
    DriverAxis=(X=1,Y=0,Z=0)
    DriverItem=(Type=Bone,Name="head")
    FalloffRegionScaleFactors=(PositiveWidth=1,NegativeWidth=1,PositiveHeight=1,NegativeHeight=1)
    FalloffSize=0.4
    OptionalParentItem=(Type=Bone,Name="spine_05")
    RotationOffset=(X=-75,Y=0,Z=10)
    ExecutePin=@SequenceAggregate.A
   )
   n2 = RigVMVariableNode() {variable=Crouch Look Up LR Hair 03}
   > Deform_GetMirrorTransform_FNC_1_1 = RigVMFunctionReferenceNode(transform=@n2.Value)
   n3 = RigVMVariableNode() {variable=Crouch Look Up LR Hair 01}
   > Deform_GetMirrorTransform_FNC_1_1_1 = RigVMFunctionReferenceNode(transform=@n3.Value)
   > &RigUnit_ModifyTransforms(
    ItemToModify="((Item=(Type=Bone),Transform=(Rotation=(X=0,Y=0,Z=0,W=1),Translation=(X=0,Y=0,Z=0),Scale3D=(X=1,Y=1,Z=1))),(Item=(Type=Bone,Name=\"dyn_rear_hair_01_root\"),Transform=(Rotation=(X=0,Y=0,Z=0,W=1),Translation=(X=0,Y=0,Z=0),Scale3D=(X=1,Y=1,Z=1))),(Item=(Type=Bo ...(truncated)"
    ItemToModify.3.Transform=@Deform_GetMirrorTransform_FNC_1_1.mirrorTransform
    ItemToModify.5.Transform=@Deform_GetMirrorTransform_FNC_1_1_1.mirrorTransform
    Weight=@n1.OutputParam
    ItemToModify.1.Transform=RigVMVariableNode() {variable=Crouch Look Up Rear Hair}.Value
    ItemToModify.2.Transform=@n2.Value
    ItemToModify.4.Transform=@n3.Value
   )
   n4 = &RigUnit_SphericalPoseReader(
    ActiveRegionScaleFactors=(PositiveWidth=2,NegativeWidth=3,PositiveHeight=2,NegativeHeight=2)
    ActiveRegionSize=0.05
    Debug=(bDrawDebug=false,DebugScale=25,DebugSegments=20,DebugThickness=0.25)
    DriverAxis=(X=1,Y=0,Z=0)
    DriverItem=(Type=Bone,Name="head")
    FalloffRegionScaleFactors=(PositiveWidth=1,NegativeWidth=1,PositiveHeight=1,NegativeHeight=1)
    FalloffSize=0.3
    OptionalParentItem=(Type=Bone,Name="spine_05")
    RotationOffset=(X=-30,Y=-15,Z=0)
    ExecutePin=@SequenceAggregate.B
   )
   > &RigUnit_ModifyTransforms(
    ItemToModify="((Item=(Type=Bone),Transform=(Rotation=(X=0,Y=0,Z=0,W=1),Translation=(X=0,Y=0,Z=0),Scale3D=(X=1,Y=1,Z=1))),(Item=(Type=Bone,Name=\"dyn_rear_hair_01_root\"),Transform=(Rotation=(X=0,Y=0,Z=0,W=1),Translation=(X=0,Y=0,Z=0),Scale3D=(X=1,Y=1,Z=1))),(Item=(Type=Bo ...(truncated)"
    Weight=@n4.OutputParam
    ItemToModify.3.Transform=RigVMVariableNode() {variable=Crouch ADS Up R Hair 01}.Value
    ItemToModify.5.Transform=RigVMVariableNode() {variable=Crouch ADS Up L Hair 01}.Value
    ItemToModify.4.Transform=RigVMVariableNode() {variable=Crouch ADS Up L Hair 03}.Value
    ItemToModify.1.Transform=RigVMVariableNode() {variable=Crouch ADS Up Rear Hair}.Value
    ItemToModify.2.Transform=RigVMVariableNode() {variable=Crouch ADS Up R Hair 03}.Value
   )
   RigVMFunctionReturnNode(ExecuteContext=@SequenceAggregate.D)
  ]}
 ContainedGraph {nodes [
   n1 = RigVMFunctionEntryNode()
   SequenceAggregate = &RigVMFunction_Sequence(ExecuteContext=@n1.ExecuteContext)
   SequenceAggregate_1 = &RigVMFunction_Sequence(ExecuteContext=@SequenceAggregate.B)
   SequenceAggregate_2 = &RigVMFunction_Sequence(ExecuteContext=@SequenceAggregate_1.B)
   RigVMFunctionReturnNode(A=@SequenceAggregate.A, B=@SequenceAggregate_1.A, C=@SequenceAggregate_2.A, D=@SequenceAggregate_2.B)
  ]}
 ContainedGraph {nodes [
   RigVMFunctionEntryNode()
   > n1 = &RigUnit_SphericalPoseReader(
    ActiveRegionScaleFactors=(PositiveWidth=1,NegativeWidth=3,PositiveHeight=1,NegativeHeight=1)
    ActiveRegionSize=0.05
    Debug=(bDrawDebug=true,DebugScale=25,DebugSegments=20,DebugThickness=0.25)
    DriverAxis=(X=1,Y=0,Z=0)
    DriverItem=(Type=Bone,Name="head")
    FalloffRegionScaleFactors=(PositiveWidth=1,NegativeWidth=1,PositiveHeight=1,NegativeHeight=1)
    FalloffSize=0.3
    OptionalParentItem=(Type=Bone,Name="pelvis")
    RotationOffset=(X=-90,Y=0,Z=20)
   )
   n2 = RigVMVariableNode() {variable=Skydiving LR Hair 03}
   > Deform_GetMirrorTransform_FNC_1_1 = RigVMFunctionReferenceNode(transform=@n2.Value)
   n3 = RigVMVariableNode() {variable=Skydiving LR Hair 01}
   > Deform_GetMirrorTransform_FNC_1_1_1 = RigVMFunctionReferenceNode(transform=@n3.Value)
   > &RigUnit_ModifyTransforms(
    ItemToModify="((Item=(Type=Bone),Transform=(Rotation=(X=0,Y=0,Z=0,W=1),Translation=(X=0,Y=0,Z=0),Scale3D=(X=1,Y=1,Z=1))),(Item=(Type=Bone,Name=\"dyn_rear_hair_01_root\"),Transform=(Rotation=(X=0,Y=0,Z=0,W=1),Translation=(X=0,Y=0,Z=0),Scale3D=(X=1,Y=1,Z=1))),(Item=(Type=Bo ...(truncated)"
    ItemToModify.3.Transform=@Deform_GetMirrorTransform_FNC_1_1.mirrorTransform
    ItemToModify.5.Transform=@Deform_GetMirrorTransform_FNC_1_1_1.mirrorTransform
    Weight=@n1.OutputParam
    ItemToModify.1.Transform=RigVMVariableNode() {variable=Skydiving Rear Hair}.Value
    ItemToModify.2.Transform=@n2.Value
    ItemToModify.4.Transform=@n3.Value
   )
   > RigVMFunctionReturnNode()
  ]}
 RigVMFunctionLibrary {nodes [
   Deform_GetMirrorTransform_FNC = RigVMCollapseNode(ExecuteContext=(), transform=(Rotation=(X=0,Y=0,Z=0,W=1),Translation=(X=0,Y=0,Z=0),Scale3D=(X=1,Y=1,Z=1)))
   Hair = RigVMCollapseNode(ExecuteContext=())
   Hair_Crouching = RigVMCollapseNode(ExecuteContext=())
   Skydiving = RigVMCollapseNode(ExecuteContext=())
  ]}
 RigVMModel {nodes [
   RigUnit_BeginExecution = &RigUnit_BeginExecution()
   Branch_RigVMUnitNode = &RigVMFunction_ControlFlowBranch(ExecuteContext=@RigUnit_BeginExecution.ExecutePin, Condition=RigVMVariableNode() {variable=IsCrouching}.Value)
   Branch_1_RigVMUnitNode = &RigVMFunction_ControlFlowBranch(ExecuteContext=@Branch_RigVMUnitNode.False, Condition=RigVMVariableNode() {variable=IsSkydiving}.Value)
   Hair = RigVMFunctionReferenceNode(ExecuteContext=@Branch_1_RigVMUnitNode.False)
   Skydiving = RigVMFunctionReferenceNode(ExecuteContext=@Branch_1_RigVMUnitNode.True)
   Hair_Crouching = RigVMFunctionReferenceNode(ExecuteContext=@Branch_RigVMUnitNode.True)
   &RigUnit_GetTransform(Item=(Type=Bone,Name="dyn_r_hair_01_root"))
   &RigUnit_GetTransform(Item=(Type=Bone,Name="dyn_r_hair_03_root"))
   RigUnit_GetTransform = &RigUnit_GetTransform(Item=(Type=Control,Name="head_ctrl"))
   RigUnit_SetTransform = &RigUnit_SetTransform(ExecutePin=(), Item=(Type=Bone,Name="head"), Weight=1, bInitial=False, Value=@RigUnit_GetTransform.Transform)
  ]}
]
hierarchy [
 root {loc=<0 0 0>, rot=<0 0 0 1>}
 pelvis {parent=root, loc=<0 -0.7071 93.1191>, rot=<0.0127 0.707 -0.0127 -0.707>}
 spine_01 {parent=pelvis, loc=<6.7494 -0.6388 0>, rot=<0 0 0.034 -0.9994>}
 spine_02 {parent=spine_01, loc=<7.2518 -0.035 0>, rot=<0 0 0.0238 -0.9997>}
 spine_03 {parent=spine_02, loc=<6.0104 0 0>, rot=<0 0 0.0149 -0.9999>}
 spine_04 {parent=spine_03, loc=<8.1024 0 0>, rot=<0 0 -0.0132 -0.9999>}
 spine_05 {parent=spine_04, loc=<10.0146 0 0>, rot=<0 0 -0.0015 -1>}
 neck_01 {parent=spine_05, loc=<11.3138 0.1529 0>, rot=<0 0 -0.2003 -0.9797>}
 neck_02 {parent=neck_01, loc=<4.5627 0 0>, rot=<0 0 0.0504 -0.9987>}
 head {parent=neck_02, loc=<5.3914 0 0>, rot=<0 0 0.115 -0.9934>}
 dyn_l_hair_01_root {parent=head, loc=<3.604 -2.156 -7.49>, rot=<0.1419 -0.7223 0.6768 -0.0057>}
 dyn_l_hair_01 {parent=dyn_l_hair_01_root, loc=<0 0 0>, rot=<0 0 0 1>}
 dyn_l_hair_02 {parent=dyn_l_hair_01, loc=<5.615 -0.0002 0.0004>, rot=<0 0 -0.0963 -0.9954>}
 dyn_r_hair_01_root {parent=head, loc=<3.605 -2.1408 7.4899>, rot=<0.7223 0.1419 -0.0057 -0.6768>}
 dyn_r_hair_01 {parent=dyn_r_hair_01_root, loc=<0 0 0>, rot=<0 0 0 1>}
 dyn_r_hair_02 {parent=dyn_r_hair_01, loc=<-5.6194 0.0005 -0.015>, rot=<0 0 -0.0963 -0.9954>}
 dyn_rear_hair_01_root {parent=head, loc=<4.288 -7.5508 -0.0851>, rot=<0.0086 0.0093 0.9986 -0.0507>}
 dyn_rear_hair_01 {parent=dyn_rear_hair_01_root, loc=<0 0 0>, rot=<0 0 0 1>}
 dyn_rear_hair_02 {parent=dyn_rear_hair_01, loc=<6.4862 0.0138 -0.0003>, rot=<0 0 -0.038 -0.9993>}
 dyn_l_fringe_root {parent=head, loc=<5.382 3.2891 -6.5196>, rot=<0.124 -0.7357 0.6523 -0.1334>}
 dyn_l_fringe {parent=dyn_l_fringe_root, loc=<0 0 0>, rot=<0 0 -0.2338 -0.9723>}
 dyn_r_fringe_root {parent=head, loc=<6.9361 3.8829 6.4621>, rot=<0.7357 0.124 -0.1334 -0.6523>}
 dyn_r_fringe {parent=dyn_r_fringe_root, loc=<0 0 0>, rot=<0 0 0 1>}
 dyn_fringe_root {parent=head, loc=<10.0367 9.0964 -1.3368>, rot=<-0.1404 -0.0186 0.9761 -0.1646>}
 dyn_fringe {parent=dyn_fringe_root, loc=<0 0 0>, rot=<0 0 0 1>}
 dyn_l_hair_03_root {parent=head, loc=<3.542 -6.5358 -4.9931>, rot=<0.0363 -0.3242 0.9453 0.0019>}
 dyn_l_hair_03 {parent=dyn_l_hair_03_root, loc=<0 0 0>, rot=<0 0 0 1>}
 dyn_l_hair_04 {parent=dyn_l_hair_03, loc=<5.6153 0.0143 0.0111>, rot=<0 0 -0.0963 -0.9954>}
 dyn_r_hair_03_root {parent=head, loc=<3.542 -6.5358 4.9931>, rot=<0.3242 0.0363 0.0019 -0.9453>}
 dyn_r_hair_03 {parent=dyn_r_hair_03_root, loc=<0 0 0>, rot=<0 0 0 1>}
 dyn_r_hair_04 {parent=dyn_r_hair_03, loc=<-5.6154 -0.0143 -0.0111>, rot=<0 0 -0.0963 -0.9954>}
 clavicle_l {parent=spine_05, loc=<4.3575 -0.1466 -2.3223>, rot=<0.0465 -0.7289 -0.009 -0.6829>}
 clavicle_r {parent=spine_05, loc=<4.3575 -0.1466 2.3223>, rot=<0.7289 0.0465 -0.6829 0.009>}
 head_ctrl {controlType=EULER_TRANSFORM, type=Control, loc=<0 0 0>, rot=<0 0 0 1>}
]
variables [
 Crouch ADS Up L Hair 01 {type=FTransform, default=(Rotation=(X=0,Y=0,Z=0,W=1),Translation=(X=0,Y=0,Z=0),Scale3D=(X=1,Y=1,Z=1))}
 Crouch ADS Up L Hair 03 {type=FTransform, default=(Rotation=(X=0,Y=0,Z=0,W=1),Translation=(X=0,Y=0,Z=0),Scale3D=(X=1,Y=1,Z=1))}
 Crouch ADS Up R Hair 01 {type=FTransform, default=(Rotation=(X=0,Y=0,Z=0,W=1),Translation=(X=0,Y=0,Z=0),Scale3D=(X=1,Y=1,Z=1))}
 Crouch ADS Up R Hair 03 {type=FTransform, default=(Rotation=(X=0,Y=0,Z=0,W=1),Translation=(X=0,Y=0,Z=0),Scale3D=(X=1,Y=1,Z=1))}
 Crouch ADS Up Rear Hair {type=FTransform, default=(Rotation=(X=0,Y=0,Z=0,W=1),Translation=(X=0,Y=0,Z=0),Scale3D=(X=1,Y=1,Z=1))}
 Crouch Look Up LR Hair 01 {type=FTransform, default=(Rotation=(X=0,Y=0,Z=0,W=1),Translation=(X=0,Y=0,Z=0),Scale3D=(X=1,Y=1,Z=1))}
 Crouch Look Up LR Hair 03 {type=FTransform, default=(Rotation=(X=0,Y=0,Z=0,W=1),Translation=(X=0,Y=0,Z=0),Scale3D=(X=1,Y=1,Z=1))}
 Crouch Look Up Rear Hair {type=FTransform, default=(Rotation=(X=0,Y=0,Z=0,W=1),Translation=(X=0,Y=0,Z=0),Scale3D=(X=1,Y=1,Z=1))}
 IsCrouching {type=bool, default=True}
 IsSkydiving {type=bool, default=True}
 Rear_Hair_Head_Bk {type=FTransform, default=(Rotation=(X=0,Y=0,Z=0.5736,W=0.8192),Translation=(X=0,Y=0,Z=0),Scale3D=(X=1,Y=1,Z=1))}
 RotationOffset {type=FVector, default=(X=-90,Y=0,Z=25)}
 SPR Parent {type=FName, default=pelvis}
 Skydiving LR Hair 01 {type=FTransform, default=(Rotation=(X=0,Y=0,Z=0,W=1),Translation=(X=0,Y=0,Z=0),Scale3D=(X=1,Y=1,Z=1))}
 Skydiving LR Hair 03 {type=FTransform, default=(Rotation=(X=0,Y=0,Z=0,W=1),Translation=(X=0,Y=0,Z=0),Scale3D=(X=1,Y=1,Z=1))}
 Skydiving Rear Hair {type=FTransform, default=(Rotation=(X=0,Y=0,Z=0,W=1),Translation=(X=0,Y=0,Z=0),Scale3D=(X=1,Y=1,Z=1))}
 UseHeadBackSPR {type=double, default=1}
]
