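# Physics asset for the PlotTwist medium male player's FaceAcc mesh part, giving capsule bodies to its dynamic fringe, side and rear hair bones.
# Hair chains hang from small root bodies joined to the head with free angular motion, while the segments below have tight angular limits.
# Free cross-links tie rear hair to both lower side chains and join two right-side chains.
# Head, neck, upper spine and clavicle capsules supply the body volumes.
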PhysicsAsset /Game/Characters/Player/Male/Medium/Bodies/M_MED_PlotTwist/Meshes/Parts/M_MED_PlotTwist_FaceAcc_Physics
bodies {
 clavicle_l [{
   shapeType=Capsule
   length=26.7812
   radius=8.7238
   shapeName=clavicle_l_capsule
   center=<6.6446 0 -0.5946>
   rotation=<rot -0.0006 -90 93.8015>
  }]
 clavicle_r [{
   shapeType=Capsule
   length=25.0625
   radius=10.4425
   shapeName=clavicle_l_capsule
   center=<-6.7213 0 0.5111>
   rotation=<rot -0.0007 -90 -80.8298>
  }]
 dyn_fringe [{shapeType=Capsule, length=3.1619, radius=2.2772, shapeName=dyn_fringe_capsule, center=<2.0343 0 -0.6423>, rotation=<rot 46.9972 0 0>}]
 dyn_fringe_root [{shapeType=Capsule, length=0.505, radius=0.505, shapeName=dyn_fringe_root_capsule, rotation=<rot 0 -90 0>}]
 dyn_l_fringe [{shapeType=Capsule, length=4.1837, radius=1.7526, shapeName=dyn_l_fringe_capsule, center=<1 0 0>}]
 dyn_l_fringe_root [{shapeType=Capsule, length=0.505, radius=0.505, shapeName=dyn_l_fringe_root_capsule, rotation=<rot 0 -90 0>}]
 dyn_l_hair_01 [{
   shapeType=Capsule
   length=4.2776
   radius=2.9442
   shapeName=dyn_l_hair_01_capsule
   center=<1.427 1.3774 0.3712>
   rotation=<rot -0.0001 0 -15.0809>
  }]
 dyn_l_hair_01_root [{shapeType=Capsule, length=0.505, radius=0.505, shapeName=dyn_l_hair_01_root_capsule, rotation=<rot 0 -90 0>}]
 dyn_l_hair_02 [{
   shapeType=Capsule
   length=4.5625
   radius=1.4375
   shapeName=dyn_l_hair_02_capsule
   center=<2.5821 2.1315 -1.1724>
   rotation=<rot 0 0 -14.8165>
  }]
 dyn_l_hair_03 [{
   shapeType=Capsule
   length=3.0756
   radius=2.206
   shapeName=dyn_l_hair_03_capsule
   center=<1.0236 1.042 0.7653>
   rotation=<rot -8.7943 -84.0059 1.2582>
  }]
 dyn_l_hair_03_root [{shapeType=Capsule, length=0.505, radius=0.505, shapeName=dyn_l_hair_03_root_capsule, rotation=<rot 0 -90 0>}]
 dyn_l_hair_04 [{
   shapeType=Capsule
   length=2.5
   radius=3.5
   shapeName=dyn_l_hair_04_capsule
   center=<2.7443 2.3497 0.3206>
   rotation=<rot 0 -113.5124 -178.969>
  }]
 dyn_r_fringe [{shapeType=Capsule, length=2.6051, radius=1.8626, shapeName=dyn_r_fringe_capsule, center=<-1 0 0>}]
 dyn_r_fringe_root [{shapeType=Capsule, length=0.505, radius=0.505, shapeName=dyn_r_fringe_root_capsule, rotation=<rot 0 -90 0>}]
 dyn_r_hair_01 [{
   shapeType=Capsule
   length=4.2776
   radius=2.9442
   shapeName=dyn_l_hair_01_capsule
   center=<-1.427 -1.3774 -0.3712>
   rotation=<rot -0.0001 0 164.9191>
  }]
 dyn_r_hair_01_root [{shapeType=Capsule, length=0.505, radius=0.505, shapeName=dyn_l_hair_01_root_capsule, rotation=<rot 0 -90 180>}]
 dyn_r_hair_02 [{
   shapeType=Capsule
   length=4.5625
   radius=3.5625
   shapeName=dyn_l_hair_02_capsule
   center=<-2.9546 -4.8327 1.6452>
   rotation=<rot -5.3256 1.4012 -177.6759>
  }]
 dyn_r_hair_03 [{
   shapeType=Capsule
   length=3.0756
   radius=2.206
   shapeName=dyn_l_hair_03_capsule
   center=<-1.0236 -1.042 -0.7653>
   rotation=<rot -8.7943 -84.0059 -178.7418>
  }]
 dyn_r_hair_03_root [{shapeType=Capsule, length=0.505, radius=0.505, shapeName=dyn_r_hair_03_root_capsule, rotation=<rot 0 -90 0>}]
 dyn_r_hair_04 [{
   shapeType=Capsule
   length=5.5
   radius=3.9375
   shapeName=dyn_l_hair_04_capsule
   center=<-2.2387 -2.1138 -0.3273>
   rotation=<rot 0 -113.5124 1.031>
  }]
 dyn_rear_hair_01 [{shapeType=Capsule, length=4.2957, radius=4.0416, shapeName=dyn_rear_hair_01_capsule, center=<2.1501 1.3895 0>}]
 dyn_rear_hair_01_root [{shapeType=Capsule, length=0.505, radius=0.505, shapeName=dyn_rear_hair_01_root_capsule, rotation=<rot 0 -90 0>}]
 dyn_rear_hair_02 [{shapeType=Capsule, length=7.4688, radius=5.25, shapeName=dyn_rear_hair_02_capsule, center=<2.1693 1.7334 0>}]
 head [{shapeType=Capsule, length=0, radius=6.2097, shapeName=head_capsule, center=<8.6552 0.3427 -0.837>, rotation=<rot 0.1011 -91.1099 -86.7768>}]
 neck_02 [{
   shapeType=Capsule
   length=12.0625
   radius=7.0988
   shapeName=neck_02_capsule
   center=<0.1629 0.6418 -0.0008>
   rotation=<rot -0.0664 -75.5791 90.2581>
  }]
 spine_05 [{
   shapeType=Capsule
   length=7.8125
   radius=13.2238
   shapeName=spine_05_capsule
   center=<0.7225 0.1311 -0.006>
   rotation=<rot -0.0856 -79.7147 90.4715>
  }]
}
constraints [
 {
  bone1Name=dyn_fringe_root
  bone2Name=head
  swing1LimitDegrees=45
  swing1Motion=Free
  swing2LimitDegrees=45
  swing2Motion=Free
  twistLimitDegrees=45
  twistMotion=Free
 }
 {bone1Name=dyn_fringe, bone2Name=dyn_fringe_root, swing1LimitDegrees=8, swing2LimitDegrees=8, twistLimitDegrees=8}
 {
  bone1Name=dyn_l_fringe_root
  bone2Name=head
  swing1LimitDegrees=45
  swing1Motion=Free
  swing2LimitDegrees=45
  swing2Motion=Free
  twistLimitDegrees=45
  twistMotion=Free
 }
 {bone1Name=dyn_l_fringe, bone2Name=dyn_l_fringe_root, swing1LimitDegrees=8, swing2LimitDegrees=8, twistLimitDegrees=8}
 {
  bone1Name=dyn_l_hair_01_root
  bone2Name=head
  swing1LimitDegrees=45
  swing1Motion=Free
  swing2LimitDegrees=45
  swing2Motion=Free
  twistLimitDegrees=45
  twistMotion=Free
 }
 {bone1Name=dyn_l_hair_01, bone2Name=dyn_l_hair_01_root, swing1LimitDegrees=8, swing2LimitDegrees=8, twistLimitDegrees=4}
 {bone1Name=dyn_l_hair_02, bone2Name=dyn_l_hair_01, swing1LimitDegrees=8, swing2LimitDegrees=8, twistLimitDegrees=4}
 {
  bone1Name=dyn_r_fringe_root
  bone2Name=head
  swing1LimitDegrees=45
  swing1Motion=Free
  swing2LimitDegrees=45
  swing2Motion=Free
  twistLimitDegrees=45
  twistMotion=Free
 }
 {bone1Name=dyn_r_fringe, bone2Name=dyn_r_fringe_root, swing1LimitDegrees=8, swing2LimitDegrees=8, twistLimitDegrees=8}
 {
  bone1Name=dyn_rear_hair_01_root
  bone2Name=head
  swing1LimitDegrees=45
  swing1Motion=Free
  swing2LimitDegrees=45
  swing2Motion=Free
  twistLimitDegrees=45
  twistMotion=Free
 }
 {bone1Name=dyn_rear_hair_01, bone2Name=dyn_rear_hair_01_root, swing1LimitDegrees=4, swing2LimitDegrees=4, twistLimitDegrees=4}
 {bone1Name=dyn_rear_hair_02, bone2Name=dyn_rear_hair_01, swing1LimitDegrees=8, swing2LimitDegrees=8, twistLimitDegrees=8}
 {
  bone1Name=dyn_r_hair_01_root
  bone2Name=head
  swing1LimitDegrees=45
  swing1Motion=Free
  swing2LimitDegrees=45
  swing2Motion=Free
  twistLimitDegrees=45
  twistMotion=Free
 }
 {bone1Name=dyn_r_hair_02, bone2Name=dyn_r_hair_01, swing1LimitDegrees=8, swing2LimitDegrees=8, twistLimitDegrees=4}
 {bone1Name=dyn_r_hair_01, bone2Name=dyn_r_hair_01_root, swing1LimitDegrees=8, swing2LimitDegrees=8, twistLimitDegrees=4}
 {
  bone1Name=dyn_l_hair_03_root
  bone2Name=head
  swing1LimitDegrees=45
  swing1Motion=Free
  swing2LimitDegrees=45
  swing2Motion=Free
  twistLimitDegrees=45
  twistMotion=Free
 }
 {bone1Name=dyn_l_hair_03, bone2Name=dyn_l_hair_03_root, swing1LimitDegrees=8, swing2LimitDegrees=8, twistLimitDegrees=8}
 {bone1Name=dyn_l_hair_04, bone2Name=dyn_l_hair_03, swing1LimitDegrees=8, swing2LimitDegrees=8, twistLimitDegrees=8}
 {
  bone1Name=dyn_r_hair_03_root
  bone2Name=head
  swing1LimitDegrees=45
  swing1Motion=Free
  swing2LimitDegrees=45
  swing2Motion=Free
  twistLimitDegrees=45
  twistMotion=Free
 }
 {bone1Name=dyn_r_hair_03, bone2Name=dyn_r_hair_03_root, swing1LimitDegrees=4, swing2LimitDegrees=4, twistLimitDegrees=4}
 {bone1Name=dyn_r_hair_04, bone2Name=dyn_r_hair_03, swing1LimitDegrees=8, swing2LimitDegrees=8, twistLimitDegrees=8}
 {
  bone1Name=dyn_rear_hair_02
  bone2Name=dyn_r_hair_04
  swing1LimitDegrees=45
  swing1Motion=Free
  swing2LimitDegrees=45
  swing2Motion=Free
  twistLimitDegrees=45
  twistMotion=Free
 }
 {
  bone1Name=dyn_rear_hair_02
  bone2Name=dyn_l_hair_04
  swing1LimitDegrees=45
  swing1Motion=Free
  swing2LimitDegrees=45
  swing2Motion=Free
  twistLimitDegrees=45
  twistMotion=Free
 }
 {
  bone1Name=head
  bone2Name=neck_02
  swing1LimitDegrees=45
  swing1Motion=Free
  swing2LimitDegrees=45
  swing2Motion=Free
  twistLimitDegrees=45
  twistMotion=Free
 }
 {
  bone1Name=neck_02
  bone2Name=spine_05
  swing1LimitDegrees=45
  swing1Motion=Free
  swing2LimitDegrees=45
  swing2Motion=Free
  twistLimitDegrees=45
  twistMotion=Free
 }
 {
  bone1Name=clavicle_l
  bone2Name=spine_05
  swing1LimitDegrees=45
  swing1Motion=Free
  swing2LimitDegrees=45
  swing2Motion=Free
  twistLimitDegrees=45
  twistMotion=Free
 }
 {
  bone1Name=clavicle_r
  bone2Name=spine_05
  swing1LimitDegrees=45
  swing1Motion=Free
  swing2LimitDegrees=45
  swing2Motion=Free
  twistLimitDegrees=45
  twistMotion=Free
 }
 {
  bone1Name=dyn_r_hair_02
  bone2Name=dyn_r_hair_04
  swing1LimitDegrees=45
  swing1Motion=Free
  swing2LimitDegrees=45
  swing2Motion=Free
  twistLimitDegrees=45
  twistMotion=Free
 }
]
