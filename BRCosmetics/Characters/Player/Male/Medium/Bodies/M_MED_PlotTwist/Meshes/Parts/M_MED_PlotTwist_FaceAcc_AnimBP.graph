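# Post-process style Anim Blueprint for the PlotTwist male face accessory, posing and simulating its hair bones.
# It copies the attached parent's pose, runs a Control Rig fed crouch and skydiving corrective transforms, then applies head-relative RigidBody physics.
# Each update, a deprecated helper computes per-movement-state gravity and wind into Gravity Override, which has no visible link to RigidBody's fixed gravity.
# Rotation-offset, SPR-parent and head-back setters are disconnected leftovers.
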AnimBlueprint /BRCosmetics/Characters/Player/Male/Medium/Bodies/M_MED_PlotTwist/Meshes/Parts/M_MED_PlotTwist_FaceAcc_AnimBP
target_skeleton=/Game/Characters/Player/Male/Medium/Bodies/M_MED_PlotTwist/Meshes/Parts/M_MED_PlotTwist_FaceAcc_Skeleton
graphs [
 AnimGraph {nodes [
   n1 = Control Rig!(
    Source=Copy Pose From Mesh() {bUseAttachedParent=true}
    Alpha
    IsCrouching:bool=Get bIsCrouching()
    Crouch ADS Up Rear Hair:Transform=Get Crouch ADS Up Rear Hair() -> (Crouch ADS Up Rear Hair:Transform)
    Crouch ADS Up R Hair 03:Transform=Get Crouch ADS Up R Hair 03() -> (Crouch ADS Up R Hair 03:Transform)
    Crouch ADS Up R Hair 01:Transform=Get Crouch ADS Up R Hair 01() -> (Crouch ADS Up R Hair 01:Transform)
    Crouch ADS Up L Hair 01:Transform=Get Crouch ADS Up L Hair 01() -> (Crouch ADS Up L Hair 01:Transform)
    Crouch ADS Up L Hair 03:Transform=Get Crouch ADS Up L Hair 03() -> (Crouch ADS Up L Hair 03:Transform)
    Crouch Look Up Rear Hair:Transform=Get Crouch Look Up Rear Hair() -> (Crouch Look Up Rear Hair:Transform)
    Crouch Look Up LR Hair 03:Transform=Get Crouch Look Up LR Hair 03() -> (Crouch Look Up LR Hair 03:Transform)
    Crouch Look Up LR Hair 01:Transform=Get Crouch Look Up LR Hair 01() -> (Crouch Look Up LR Hair 01:Transform)
    IsSkydiving:bool=Get bIsSkyDiving()
    Skydiving Rear Hair:Transform=Get Skydiving Rear Hair() -> (Skydiving Rear Hair:Transform)
    Skydiving LR Hair 03:Transform=Get Skydiving LR Hair 03() -> (Skydiving LR Hair 03:Transform)
    Skydiving LR Hair 01:Transform=Get Skydiving LR Hair 01() -> (Skydiving LR Hair 01:Transform)
   ) -> (Pose) {
    alphaBoolBlend {blendOption=Linear}
    controlRigAssetReference {blueprintRigClass=/Game/Characters/Player/Male/Medium/Bodies/M_MED_PlotTwist/Meshes/Parts/M_MED_PlotTwist_FaceAcc_ControlRig.M_MED_PlotTwist_FaceAcc_ControlRig_C}
   }
   n2 = RigidBody!(ComponentPose=Local To Component(LocalPose=@n1), Alpha, OverrideWorldGravity:Vector="0,0,-980", ExternalForce) -> (Pose) {
    bOverrideWorldGravity=true
    overridePhysicsAsset=/Game/Characters/Player/Male/Medium/Bodies/M_MED_PlotTwist/Meshes/Parts/M_MED_PlotTwist_FaceAcc_Physics
    simulationSpace=BaseBoneSpace
    alphaBoolBlend {blendOption=Linear}
    baseBoneRef {boneName=head}
    componentAppliedLinearAccClamp=<3000 3000 3000>
    componentLinearAccScale=<0.5 0.5 0.5>
    componentLinearVelScale=<0.5 0.5 0.5>
    overrideWorldGravity {z=-980}
    simSpaceSettings {maxAngularAcceleration=100, maxAngularVelocity=100, maxLinearAcceleration=100, maxLinearVelocity=100, worldAlpha=1}
   }
   Output Pose(Result=Component To Local(ComponentPose=@n2)) {layerGroup=None}
   Get Use Head Back SPR() -> (Use Head Back SPR:double)
  ]}
 EventGraph {nodes [
   Event BlueprintUpdateAnimation()
   > n1 = Gravity Override State Function DEPRECATED!(
    execute
    CharacterPartAnimInstance=Self-Reference()
    GameTimeInSeconds=GetGameTimeInSeconds()
    GravityOverrideParams_Skydive_GoingUp_28_#1:Gravity Override Setting Struct Structure=(WindFrequency_28_#2=(X=1.5,Y=7,Z=8),WindAmplitude_31_#3=(X=200,Y=-500,Z=400),WindOffset_32_#4=(X=0,Y=0,Z=0),GravityOverride_15_#5=(X=0,Y=-300,Z=1500),JointName_27_#6="root")
    GravityOverrideParams_Skydive_FloatingDown_29_#7:Gravity Override Setting Struct Structure=Get Gravity Override Params Skydive Floating Down() -> (Gravity Override Params Skydive Floating Down:Gravity Override Setting Struct Structure)
    GravityOverrideParams_Skydive_DivingDown_30_#8:Gravity Override Setting Struct Structure=(WindFrequency_28_#2=(X=1.5,Y=12,Z=10),WindAmplitude_31_#3=(X=50,Y=300,Z=100),WindOffset_32_#4=(X=0,Y=0,Z=0),GravityOverride_15_#5=(X=-1500,Y=0,Z=0),JointName_27_#6="pelvis")
    GravityOverrideParams_Skydive_Parachuting_31_#9:Gravity Override Setting Struct Structure=Get Gravity Override Params Skydive Parachuting() -> (Gravity Override Params Skydive Parachuting:Gravity Override Setting Struct Structure)
    GravityOverrideParams_OnGround_Standing_32_#10:Gravity Override Setting Struct Structure=(WindFrequency_28_#2=(X=0,Y=0,Z=0),WindAmplitude_31_#3=(X=0,Y=0,Z=0),WindOffset_32_#4=(X=0,Y=0,Z=0),GravityOverride_15_#5=(X=0,Y=0,Z=-980),JointName_27_#6="root")
    GravityOverrideParams_OnGround_Moving_33_#11:Gravity Override Setting Struct Structure=(WindFrequency_28_#2=(X=0,Y=0,Z=0),WindAmplitude_31_#3=(X=0,Y=0,Z=0),WindOffset_32_#4=(X=0,Y=0,Z=0),GravityOverride_15_#5=(X=0,Y=0,Z=-980),JointName_27_#6="root")
    GravityOverrideParams_EmoteOrMelee_34_#12:Gravity Override Setting Struct Structure=(WindFrequency_28_#2=(X=0,Y=0,Z=0),WindAmplitude_31_#3=(X=0,Y=0,Z=0),WindOffset_32_#4=(X=0,Y=0,Z=0),GravityOverride_15_#5=(X=0,Y=0,Z=-980),JointName_27_#6="root")
    GravityOverrideParams_Swimming_35_#13:Gravity Override Setting Struct Structure=(WindFrequency_28_#2=(X=0,Y=0,Z=0),WindAmplitude_31_#3=(X=0,Y=0,Z=0),WindOffset_32_#4=(X=0,Y=0,Z=0),GravityOverride_15_#5=(X=-800,Y=0,Z=0),JointName_27_#6="pelvis")
    GravityOverrideParams_Front_End_36_#14:Gravity Override Setting Struct Structure=(WindFrequency_28_#2=(X=0,Y=0,Z=0),WindAmplitude_31_#3=(X=0,Y=0,Z=0),WindOffset_32_#4=(X=0,Y=0,Z=0),GravityOverride_15_#5=(X=0,Y=0,Z=-980),JointName_27_#6="root")
    GravityOverrideParams_DBNO_40_#15:Gravity Override Setting Struct Structure=(WindFrequency_28_#2=(X=0,Y=0,Z=0),WindAmplitude_31_#3=(X=0,Y=0,Z=0),WindOffset_32_#4=(X=0,Y=0,Z=0),GravityOverride_15_#5=(X=-980,Y=400,Z=0),JointName_27_#6="pelvis")
    GravityOverrideParams_DBNO_Carried_41_#16:Gravity Override Setting Struct Structure=(WindFrequency_28_#2=(X=0,Y=0,Z=0),WindAmplitude_31_#3=(X=0,Y=0,Z=0),WindOffset_32_#4=(X=0,Y=0,Z=0),GravityOverride_15_#5=(X=-980,Y=500,Z=-500),JointName_27_#6="pelvis")
    GravityOverrideParams_OnGround_Falling_45_#17:Gravity Override Setting Struct Structure=(WindFrequency_28_#2=(X=0,Y=0,Z=0),WindAmplitude_31_#3=(X=0,Y=0,Z=0),WindOffset_32_#4=(X=0,Y=0,Z=0),GravityOverride_15_#5=(X=0,Y=0,Z=-980),JointName_27_#6="root")
    GravityOverrideParams_Ziplining_46_#18:Gravity Override Setting Struct Structure=(WindFrequency_28_#2=(X=0,Y=0,Z=0),WindAmplitude_31_#3=(X=0,Y=0,Z=0),WindOffset_32_#4=(X=0,Y=0,Z=0),GravityOverride_15_#5=(X=0,Y=0,Z=-980),JointName_27_#6="root")
    GravityOverrideParams_Skydive_Surfing_49_#19:Gravity Override Setting Struct Structure=(WindFrequency_28_#2=(X=1.5,Y=7,Z=8),WindAmplitude_31_#3=(X=200,Y=-500,Z=400),WindOffset_32_#4=(X=0,Y=0,Z=0),GravityOverride_15_#5=(X=500,Y=-300,Z=-980),JointName_27_#6="root")
    GravityOverrideParams_Riding_Standing_54_#20:Gravity Override Setting Struct Structure=(WindFrequency_28_#2=(X=0,Y=0,Z=0),WindAmplitude_31_#3=(X=0,Y=0,Z=0),WindOffset_32_#4=(X=0,Y=0,Z=0),GravityOverride_15_#5=(X=0,Y=-350,Z=-980),JointName_27_#6="root")
    GravityOverrideParams_Riding_Moving_55_#21:Gravity Override Setting Struct Structure=(WindFrequency_28_#2=(X=0,Y=0,Z=0),WindAmplitude_31_#3=(X=0,Y=0,Z=0),WindOffset_32_#4=(X=0,Y=0,Z=0),GravityOverride_15_#5=(X=0,Y=-450,Z=-450),JointName_27_#6="root")
    GravityOverrideParams_Riding_Moving_Targeting_58_#22:Gravity Override Setting Struct Structure=(WindFrequency_28_#2=(X=0,Y=0,Z=0),WindAmplitude_31_#3=(X=0,Y=0,Z=0),WindOffset_32_#4=(X=0,Y=0,Z=0),GravityOverride_15_#5=(X=0,Y=-200,Z=-980),JointName_27_#6="root")
    GravityOverrideParams_Tactical_Sprint_57_#23:Gravity Override Setting Struct Structure=(WindFrequency_28_#2=(X=0,Y=0,Z=0),WindAmplitude_31_#3=(X=0,Y=0,Z=0),WindOffset_32_#4=(X=0,Y=0,Z=0),GravityOverride_15_#5=(X=0,Y=-450,Z=-500),JointName_27_#6="root")
    GravityOverrideParams_Rail_Grinding_61_#24:Gravity Override Setting Struct Structure=(WindFrequency_28_#2=(X=0,Y=0,Z=0),WindAmplitude_31_#3=(X=0,Y=0,Z=0),WindOffset_32_#4=(X=0,Y=0,Z=0),GravityOverride_15_#5=(X=0,Y=0,Z=-980),JointName_27_#6="root")
    GravityOverrideParams_DBNO_Tumbling_67_#25:Gravity Override Setting Struct Structure=(WindFrequency_28_#2=(X=0,Y=0,Z=0),WindAmplitude_31_#3=(X=0,Y=0,Z=0),WindOffset_32_#4=(X=0,Y=0,Z=0),GravityOverride_15_#5=(X=-2000,Y=0,Z=0),JointName_27_#6="pelvis")
    GravityOverrideParams_DBNO_RollFwd_68_#26:Gravity Override Setting Struct Structure=(WindFrequency_28_#2=(X=0,Y=0,Z=0),WindAmplitude_31_#3=(X=0,Y=0,Z=0),WindOffset_32_#4=(X=0,Y=0,Z=0),GravityOverride_15_#5=(X=-2000,Y=0,Z=0),JointName_27_#6="pelvis")
    GravityOverrideParams_DBNO_RollBwd_69_#27:Gravity Override Setting Struct Structure=(WindFrequency_28_#2=(X=0,Y=0,Z=0),WindAmplitude_31_#3=(X=0,Y=0,Z=0),WindOffset_32_#4=(X=0,Y=0,Z=0),GravityOverride_15_#5=(X=-4000,Y=0,Z=0),JointName_27_#6="pelvis")
    GravityOverrideParams_DBNO_Impulse_70_#28:Gravity Override Setting Struct Structure=(WindFrequency_28_#2=(X=0,Y=0,Z=0),WindAmplitude_31_#3=(X=0,Y=0,Z=0),WindOffset_32_#4=(X=0,Y=0,Z=0),GravityOverride_15_#5=(X=-2000,Y=0,Z=0),JointName_27_#6="pelvis")
    RBAN Params
    DebugDraw
    DebugJointName
   ) -> (then, GravityOverride, CurrentStateOut, GravityOverrideSettingsOut, RBAN SettingsOut)
   > Set Gravity Override(Gravity Override=@n1)
   n2 = Branch()
   > Set Rotation Offset(Rotation Offset:Vector="-90,0,-20") -> (Output_Get:Vector)
   > Set SPR Parent(SPR Parent:Name=pelvis) -> (Output_Get:Name=None)
   Set Rotation Offset(execute=@n2.else, Rotation Offset="-90,0,20")
   > Set SPR Parent(SPR Parent=neck_02) -> (Output_Get=None)
   Get bIsSkyDiving()
   Set Use Head Back SPR(Use Head Back SPR:double=0) -> (Output_Get:double)
   Set Use Head Back SPR(Use Head Back SPR=1)
  ]}
]
variables [
 Crouch ADS Up L Hair 01 {type=Transform, default {location=<0 0 0>, rotation=<rot 0 0 0>, scale=<1 1 1>}}
 Crouch ADS Up L Hair 03 {type=Transform, default {location=<0 0 0>, rotation=<rot 10 0 0>, scale=<1 1 1>}}
 Crouch ADS Up R Hair 01 {type=Transform, default {location=<2 2 0>, rotation=<rot 20 50 0>, scale=<1 1 1>}}
 Crouch ADS Up R Hair 03 {type=Transform, default {location=<3 6 0>, rotation=<rot 0 45 0>, scale=<1 1 1>}}
 Crouch ADS Up Rear Hair {type=Transform, default {location=<0 -2 0>, rotation=<rot 30 60 0>, scale=<1 1 1>}}
 Crouch Look Up LR Hair 01 {type=Transform, default {location=<0 -3 0>, rotation=<rot 35 45 0>, scale=<1 1 1>}}
 Crouch Look Up LR Hair 03 {type=Transform, default {location=<-3 -3 0>, rotation=<rot 35 45 0>, scale=<1 1 1>}}
 Crouch Look Up Rear Hair {type=Transform, default {location=<-3 5 0>, rotation=<rot 0 40 0>, scale=<1 1 1>}}
 Gravity Override {type=Vector}
 Gravity Override Params Skydive Floating Down {
  type=GravityOverrideSettingStruct
  default {jointName=root, gravityOverride=<0 -700 600>, windAmplitude=<200 -500 400>, windFrequency=<1.5 7 8>, windOffset=<0 -600 400>}
 }
 Gravity Override Params Skydive Parachuting {
  type=GravityOverrideSettingStruct
  default {jointName=root, gravityOverride=<0 -800 300>, windAmplitude=<200 -500 400>, windFrequency=<1.5 7 8>, windOffset=<0 0 0>}
 }
 Rotation Offset {type=Vector, default=<-90 0 20>}
 Skydiving LR Hair 01 {type=Transform, default {location=<0 0 -3>, rotation=<rot 30 20 0>, scale=<1 1 1>}}
 Skydiving LR Hair 03 {type=Transform, default {location=<0 -2 0>, rotation=<rot 0 45 0>, scale=<1 1 1>}}
 Skydiving Rear Hair {type=Transform, default {location=<0 0 0>, rotation=<rot 0 90 0>, scale=<1 1 1>}}
 SPR Parent {type=string, default=neck_02}
 Use Head Back SPR {type=number, default=1}
]
